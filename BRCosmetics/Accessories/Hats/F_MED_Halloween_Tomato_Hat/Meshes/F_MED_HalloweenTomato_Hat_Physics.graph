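# Physics asset for the female medium Halloween Tomato hat cosmetic, giving its hood bones physics bodies.
# Six identical small capsules cover the front, top, left, right and two back dyn_hood bones, alongside a larger capsule on the head.
# Each hood body is constrained directly to the head in a star layout, with tight, equal swing and twist limits that allow only slight movement.
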PhysicsAsset /BRCosmetics/Accessories/Hats/F_MED_Halloween_Tomato_Hat/Meshes/F_MED_HalloweenTomato_Hat_Physics
bodies {
 dyn_hood_bk_l [{shapeType=Capsule, length=5, radius=5}]
 dyn_hood_bk_r [{shapeType=Capsule, length=5, radius=5}]
 dyn_hood_fr [{shapeType=Capsule, length=5, radius=5}]
 dyn_hood_l [{shapeType=Capsule, length=5, radius=5}]
 dyn_hood_r [{shapeType=Capsule, length=5, radius=5}]
 dyn_hood_top [{shapeType=Capsule, length=5, radius=5}]
 head [{shapeType=Capsule, length=1.781, radius=8.8597}]
}
constraints [
 {bone1Name=dyn_hood_top, bone2Name=head, swing1LimitDegrees=2, swing2LimitDegrees=2, twistLimitDegrees=2}
 {bone1Name=dyn_hood_l, bone2Name=head, swing1LimitDegrees=2, swing2LimitDegrees=2, twistLimitDegrees=2}
 {bone1Name=dyn_hood_bk_l, bone2Name=head, swing1LimitDegrees=2, swing2LimitDegrees=2, twistLimitDegrees=2}
 {bone1Name=dyn_hood_bk_r, bone2Name=head, swing1LimitDegrees=2, swing2LimitDegrees=2, twistLimitDegrees=2}
 {bone1Name=dyn_hood_r, bone2Name=head, swing1LimitDegrees=2, swing2LimitDegrees=2, twistLimitDegrees=2}
 {bone1Name=dyn_hood_fr, bone2Name=head, swing1LimitDegrees=2, swing2LimitDegrees=2, twistLimitDegrees=2}
]
